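# Translucent UI material that draws an animated smoky gradient background for the quest gallery, blending two dark parameterized colors.
# Two cloud patterns each multiply a pair of slowly panning samples of one blurred noise texture.
# An inverted circle SDF with a size parameter, offset and eased, masks them into a broad smoky gradient and a tighter center one.
# Named reroutes carry these gradients.
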Material /QuestCategories/UI/Materials/M_UI_Gallery_QuestBG
graph {
 properties {blendMode=BLEND_Translucent, materialDomain=MD_UI, pixelDepthOffsetMode=PDOM_Legacy, usedShadingModels=Default Lit}
 nodes [
  LinearInterpolate(
   A=VectorParameter() {group=Bkgd_Gradient, parameterName=Bkgd_Gradient_Color1, defaultValue=<rgba 0.0008 0.0008 0.0052 1>}.RGB
   B=VectorParameter() {group=Bkgd_Gradient, parameterName=Bkgd_Gradient_Color2, defaultValue=<rgba 0.0058 0.005 0.0312 1>}.RGB
   Alpha=NamedRerouteUsage()
  )
  n1 = Multiply(A=Time()) {constB=0.1}
  n2 = Panner(Time=@n1) {speedX=-0.5, speedY=-0.1}
  n3 = Panner(Time=@n1) {speedX=0.1, speedY=0.2}
  n4 = Saturate(None=Multiply(
    A=TextureSample(UVs=@n2) {samplerType=SAMPLERTYPE_Grayscale, texture=/FortUILibrary/UIKit/Buttons/Textures/LowResBlurredNoise}.R
    B=TextureSample(UVs=@n3) {samplerType=SAMPLERTYPE_Grayscale, texture=&LowResBlurredNoise}.R
   ))
  n5 = OneMinus(None=MaterialFunctionCall(Size=ScalarParameter() {defaultValue=0.625, parameterName=Inner Circle Mask Size}, GlowMax=Constant() {r=0.02}) {materialFunction=&MF_UI_Circle}.SDF)
  n6 = Multiply(A=Add(A=@n5) {constB=-0.2})
  n7 = Saturate(None=MaterialFunctionCall("Value (0-1)"=Saturate(None=@n6)) {materialFunction=&MF_UI_EaseCurves}.EaseInSine)
  n8 = Add(A=Multiply(A=NamedRerouteDeclaration(None=@n4) {name=cloudy pattern, nodeColor=<rgba 0 1 0.625 1>}, B=@n7), B=Multiply(A=@n7) {constB=0.25})
  NamedRerouteDeclaration(None=Saturate(None=Multiply(A=@n8) {constB=2})) {name=SmokeyGradient, nodeColor=<rgba 0 0 0 1>}
  n9 = Multiply(A=Time()) {constB=0.1}
  n10 = Panner(Time=@n9) {speedX=0.2, speedY=0.1}
  n11 = Panner(Time=@n9) {speedX=-0.2, speedY=-0.2}
  n12 = Saturate(None=Multiply(
    A=TextureSample(UVs=@n10) {samplerType=SAMPLERTYPE_Grayscale, texture=&LowResBlurredNoise}.R
    B=TextureSample(UVs=@n11) {samplerType=SAMPLERTYPE_Grayscale, texture=&LowResBlurredNoise}.R
   ))
  n13 = Multiply(A=Add(A=@n5) {constB=-0.66})
  n14 = Saturate(None=MaterialFunctionCall("Value (0-1)"=Saturate(None=@n13)) {materialFunction=&MF_UI_EaseCurves}.EaseInOutSine)
  n15 = Add(
   A=Multiply(A=NamedRerouteDeclaration(None=@n12) {name=cloudy pattern_2, nodeColor=<rgba 0 1 0.625 1>}, B=@n14)
   B=Multiply(A=@n14) {constB=0.5}
  )
  NamedRerouteDeclaration(None=Saturate(None=Multiply(A=@n15) {constB=2})) {name=SmokeyGradient_Center, nodeColor=<rgba 0 0 0 1>}
  Saturate(None=MaterialFunctionCall("Value (0-1)"=NamedRerouteUsage()) {materialFunction=&MF_UI_EaseCurves}.EaseOutSine)
 ]
}
